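# Material function for toon custom-depth outlines that outputs a mask hiding pixels behind the custom depth surface, as a form of backface culling.
# Pixel depth and CustomDepth are both divided by scaled camera distance to an object or particle position input, defaulting to particle position.
# An If returns 0 where pixel depth is greater, otherwise 1.
# The mobile shading path skips the test and always returns 1.
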MaterialFunction /Game/Effects/Fort_Effects/Materials/Toon/CustomDepthOutLine/MF_BackfaceCulling
graph {nodes [
  n1 = ParticlePositionWS()
  n2 = Multiply(A=Distance(A=FunctionInput(Preview=@n1) {bUsePreviewValueAsDefault=true, inputName=ObjectOrParticlePosition}, B=CameraPositionWS())) {constB=20}
  n3 = Divide(A=PixelDepth().R, B=@n2) {constB=40960}
  n4 = Divide(A=ComponentMask(None=SceneTexture() {sceneTextureId=PPI_CustomDepth}.Color) {r=true}, B=@n2) {constB=40960}
  n5 = Constant()
  n6 = Constant() {r=1}
  FunctionOutput(None=ShadingPathSwitch(Default=If(A=@n3, B=@n4, "A > B"=@n5, "A == B"=@n6, "A < B"=@n6), Mobile=Constant() {r=1}) {desc=Migrated from FeatureLevelSwitch})
 ]}
